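# Physics asset for the VelvetDeskFiber male face accessory, giving capsule bodies and joints to its dynamic hair bones around a large head capsule.
# Most hair bones hang individually off the head with identical tight swing and twist limits.
# Two front locks and a right inner strand form three-body chains from small master capsules, with limited links along each chain.
# The Front and Right masters rotate freely against the head, while the Front2 master is limited.
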PhysicsAsset /BRCosmetics/Characters/Player/Male/Medium/Bodies/M_MED_VelvetDeskFiber_FaceAcc/M_MED_VelvetDeskFiber_FaceAcc_Physics1
bodies {
 DynHair_10 [{
   shapeType=Capsule
   length=2.621
   radius=0.3694
   shapeName=DynHair_10_capsule
   center=<3.7247 -0.0316 1.5809>
   rotation=<rot 0 87.4057 -57.4665>
  }]
 Dyn_Front2_Hair1 [{shapeType=Capsule, length=3.8471, radius=0.6529, shapeName=Dyn_Front2_Hair1_capsule, center=<2 0 0>, rotation=<rot 90 0 0>}]
 Dyn_Front2_Hair2 [{
   shapeType=Capsule
   length=2.9051
   radius=0.9075
   shapeName=Dyn_Front2_Hair2_capsule
   center=<2 0.75 0>
   rotation=<rot 74.9991 15.0015 0.0002>
  }]
 Dyn_Front_Hair_01 [{shapeType=Capsule, length=3.759, radius=0.641, shapeName=Dyn_Front_Hair_01_capsule, center=<2 0 0>, rotation=<rot 90 0 0>}]
 Dyn_Front_Hair_02 [{shapeType=Capsule, length=1.9173, radius=0.9497, shapeName=Dyn_Front_Hair_02_capsule, center=<1.5423 0 -0.3>, rotation=<rot 90 0 0>}]
 Dyn_Hair2 [{shapeType=Capsule, length=2.5382, radius=1.2387, shapeName=Dyn_Hair2_capsule, center=<1.2197 0 0>, rotation=<rot 90 0 0>}]
 Dyn_Hair3 [{shapeType=Capsule, length=3.2444, radius=1.2565, shapeName=Dyn_Hair3_capsule, center=<1 0 0>, rotation=<rot 90 0 0>}]
 Dyn_Hair4 [{shapeType=Capsule, length=1.8895, radius=0.7159, shapeName=Dyn_Hair4_capsule, center=<1 0 0>, rotation=<rot 90 0 0>}]
 Dyn_Hair5 [{shapeType=Capsule, length=3.6077, radius=1.7526, shapeName=Dyn_Hair5_capsule, center=<1 0 0>, rotation=<rot 90 0 0>}]
 Dyn_Hair6 [{shapeType=Capsule, length=2.8527, radius=1.0165, shapeName=Dyn_Hair6_capsule, center=<1 0 0>, rotation=<rot 90 0 0>}]
 Dyn_Hair7 [{shapeType=Capsule, length=1, radius=0.5, shapeName=Dyn_Hair7_capsule, center=<0.5 0 0>, rotation=<rot 90 0 0>}]
 Dyn_Hair8 [{shapeType=Capsule, length=2.1254, radius=0.5, shapeName=Dyn_Hair8_capsule, center=<0.5 0 0>, rotation=<rot 90 0 0>}]
 Dyn_Hair9 [{shapeType=Capsule, length=1.6466, radius=0.5705, shapeName=Dyn_Hair9_capsule, center=<0.5 0 0>, rotation=<rot 90 0 0>}]
 Dyn_Hair_11 [{shapeType=Capsule, length=1.6941, radius=0.8103, shapeName=Dyn_Hair_11_capsule, center=<0.7 0 0>, rotation=<rot 90 0 0>}]
 Dyn_Hair_12 [{shapeType=Capsule, length=1.664, radius=1.0254, shapeName=Dyn_Hair_12_capsule, center=<1.4225 -1.8071 -1.3755>, rotation=<rot 90 0 0>}]
 Dyn_Hair_13 [{shapeType=Capsule, length=1.8672, radius=0.8736, shapeName=Dyn_Hair_13_capsule, center=<0.7 0 0>, rotation=<rot 90 0 0>}]
 Dyn_Hair_14 [{shapeType=Capsule, length=2.229, radius=0.88, shapeName=Dyn_Hair_14_capsule, center=<0.8 0 0>, rotation=<rot 90 0 0>}]
 Dyn_Hair_15 [{shapeType=Capsule, length=1.4971, radius=0.7099, shapeName=Dyn_Hair_15_capsule, center=<0.8 0 0>, rotation=<rot 90 0 0>}]
 Dyn_Hair_16 [{shapeType=Capsule, length=2.1068, radius=0.6677, shapeName=Dyn_Hair_16_capsule, center=<0.2 0 0>, rotation=<rot 90 0 0>}]
 Dyn_Hair_17 [{shapeType=Capsule, length=2.1139, radius=0.832, shapeName=Dyn_Hair_17_capsule, center=<0.7 0 0>, rotation=<rot 90 0 0>}]
 Dyn_Hair_18 [{shapeType=Capsule, length=1.6602, radius=0.6918, shapeName=Dyn_Hair_18_capsule, center=<0.7 0 0>, rotation=<rot 90 0 0>}]
 Dyn_Master_Front2_Hair [{shapeType=Capsule, length=0.5, radius=0.5, shapeName=Dyn_Master_Front2_Hair_capsule}]
 Dyn_Master_Front_Hair [{shapeType=Capsule, length=0.5, radius=0.5, shapeName=Dyn_Master_Front_Hair_capsule}]
 Dyn_Master_Right_Hairstrand_Inner1 [{shapeType=Capsule, length=0.5, radius=0.5, shapeName=Dyn_Master_Right_Hairstrand_Inner1_capsule}]
 Dyn_Right_Hairstrand_Inner1 [{shapeType=Capsule, length=3.5945, radius=1.0055, shapeName=Dyn_Right_Hairstrand_Inner1_capsule, center=<2 -0.35 0>, rotation=<rot 90 0 0>}]
 Dyn_Right_Hairstrand_Inner2 [{
   shapeType=Capsule
   length=2.2107
   radius=0.9314
   shapeName=Dyn_Right_Hairstrand_Inner2_capsule
   center=<1.1672 0.5663 0>
   rotation=<rot 90 180 155.0002>
  }]
 head [{shapeType=Capsule, length=5.39, radius=6.61, shapeName=head_capsule, center=<3.2931 3.5247 0>, rotation=<rot 90 0 0>}]
}
constraints [
 {bone1Name=DynHair_10, bone2Name=head, swing1LimitDegrees=20, swing2LimitDegrees=15, twistLimitDegrees=5}
 {bone1Name=Dyn_Hair2, bone2Name=head, swing1LimitDegrees=20, swing2LimitDegrees=15, twistLimitDegrees=5}
 {bone1Name=Dyn_Hair3, bone2Name=head, swing1LimitDegrees=20, swing2LimitDegrees=15, twistLimitDegrees=5}
 {bone1Name=Dyn_Hair4, bone2Name=head, swing1LimitDegrees=20, swing2LimitDegrees=15, twistLimitDegrees=5}
 {bone1Name=Dyn_Hair5, bone2Name=head, swing1LimitDegrees=20, swing2LimitDegrees=15, twistLimitDegrees=5}
 {bone1Name=Dyn_Hair6, bone2Name=head, swing1LimitDegrees=20, swing2LimitDegrees=15, twistLimitDegrees=5}
 {bone1Name=Dyn_Hair7, bone2Name=head, swing1LimitDegrees=20, swing2LimitDegrees=15, twistLimitDegrees=5}
 {bone1Name=Dyn_Hair8, bone2Name=head, swing1LimitDegrees=20, swing2LimitDegrees=15, twistLimitDegrees=5}
 {bone1Name=Dyn_Hair9, bone2Name=head, swing1LimitDegrees=20, swing2LimitDegrees=15, twistLimitDegrees=5}
 {bone1Name=Dyn_Hair_11, bone2Name=head, swing1LimitDegrees=20, swing2LimitDegrees=15, twistLimitDegrees=5}
 {bone1Name=Dyn_Hair_12, bone2Name=head, swing1LimitDegrees=20, swing2LimitDegrees=15, twistLimitDegrees=5}
 {bone1Name=Dyn_Hair_13, bone2Name=head, swing1LimitDegrees=20, swing2LimitDegrees=15, twistLimitDegrees=5}
 {bone1Name=Dyn_Hair_14, bone2Name=head, swing1LimitDegrees=20, swing2LimitDegrees=15, twistLimitDegrees=5}
 {bone1Name=Dyn_Hair_15, bone2Name=head, swing1LimitDegrees=20, swing2LimitDegrees=15, twistLimitDegrees=5}
 {bone1Name=Dyn_Hair_16, bone2Name=head, swing1LimitDegrees=20, swing2LimitDegrees=15, twistLimitDegrees=5}
 {bone1Name=Dyn_Hair_17, bone2Name=head, swing1LimitDegrees=20, swing2LimitDegrees=15, twistLimitDegrees=5}
 {bone1Name=Dyn_Hair_18, bone2Name=head, swing1LimitDegrees=20, swing2LimitDegrees=15, twistLimitDegrees=5}
 {bone1Name=Dyn_Master_Front2_Hair, bone2Name=head, swing1LimitDegrees=20, swing2LimitDegrees=20, twistLimitDegrees=1}
 {bone1Name=Dyn_Front2_Hair1, bone2Name=Dyn_Master_Front2_Hair, swing1LimitDegrees=20, swing2LimitDegrees=20, twistLimitDegrees=1}
 {bone1Name=Dyn_Front2_Hair2, bone2Name=Dyn_Front2_Hair1, swing1LimitDegrees=30, swing2LimitDegrees=30, twistLimitDegrees=1}
 {
  bone1Name=Dyn_Master_Front_Hair
  bone2Name=head
  swing1LimitDegrees=45
  swing1Motion=Free
  swing2LimitDegrees=45
  swing2Motion=Free
  twistLimitDegrees=45
  twistMotion=Free
 }
 {bone1Name=Dyn_Front_Hair_01, bone2Name=Dyn_Master_Front_Hair, swing1LimitDegrees=20, swing2LimitDegrees=20, twistLimitDegrees=5}
 {bone1Name=Dyn_Front_Hair_02, bone2Name=Dyn_Front_Hair_01, swing1LimitDegrees=20, swing2LimitDegrees=20, twistLimitDegrees=5}
 {
  bone1Name=Dyn_Master_Right_Hairstrand_Inner1
  bone2Name=head
  swing1LimitDegrees=45
  swing1Motion=Free
  swing2LimitDegrees=45
  swing2Motion=Free
  twistLimitDegrees=45
  twistMotion=Free
 }
 {
  bone1Name=Dyn_Right_Hairstrand_Inner1
  bone2Name=Dyn_Master_Right_Hairstrand_Inner1
  swing1LimitDegrees=20
  swing2LimitDegrees=20
  twistLimitDegrees=5
 }
 {
  bone1Name=Dyn_Right_Hairstrand_Inner2
  bone2Name=Dyn_Right_Hairstrand_Inner1
  swing1LimitDegrees=20
  swing2LimitDegrees=20
  twistLimitDegrees=5
 }
]
