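# Anim Blueprint for the VelvetDeskFiber medium male face accessory, adding physics-driven motion to its hair.
# It copies the attached parent's pose and runs RigidBody simulation in spine_05 space with scaled component motion and custom gravity.
# A per-bone blend layers DefaultSlot over front and right hair bones.
# Its weight defaults to 1 and, in the front end only, follows the Frontend_Idle_Bypass curve from the owner's anim instance.
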AnimBlueprint /BRCosmetics/Characters/Player/Male/Medium/Bodies/M_MED_VelvetDeskFiber_FaceAcc/M_MED_VelvetDeskFiber_FaceAcc_AnimBP
target_skeleton=/BRCosmetics/Characters/Player/Male/Medium/Bodies/M_MED_VelvetDeskFiber/Meshes/M_MED_VelvetDeskFiber_Skeleton
graphs [
 AnimGraph {nodes [
   n1 = RigidBody(ComponentPose=Local To Component(LocalPose=Copy Pose From Mesh() {bUseAttachedParent=true})) {
    bClampLinearTranslationLimitToRefPose=true
    bOverrideWorldGravity=true
    overridePhysicsAsset=&/M_MED_VelvetDeskFiber_FaceAcc_Physics1
    simulationSpace=BaseBoneSpace
    alphaBoolBlend {blendOption=Linear}
    baseBoneRef {boneName=spine_05}
    componentAppliedLinearAccClamp=<1000 1000 1000>
    componentLinearAccScale=<0.3 0.4 0.4>
    componentLinearVelScale=<0.3 0.4 0.4>
    overrideWorldGravity {z=-500}
   }
   n2 = Layered blend per bone(
    BasePose=Component To Local(ComponentPose=@n1)
    BlendPoses_0=Slot 'DefaultSlot'()
    BlendWeights_0=Get FE Custom Idle Is Playing Float() -> (FE Custom Idle Is Playing Float:float)
   ) {layerSetup [{branchFilters [
       {boneName=Dyn_Front_Hair_01}
       {boneName=Dyn_Front_Hair_02}
       {boneName=Dyn_Front2_Hair_1}
       {boneName=Dyn_Front2_Hair_2}
       {boneName=Dyn_Right_hairstrand_inner1}
       {boneName=Dyn_Right_hairstrand_inner2}
      ]}]}
   Output Pose(Result=@n2) {layerGroup=None}
  ]}
 EventGraph {nodes [
   Event BlueprintUpdateAnimation()
   > Branch(Condition=Get bIsInFrontEnd())
   > Set FE Custom Idle Is Playing Float(FE Custom Idle Is Playing Float:float=GetCurveValue(self=GetOwnerFrontEndAnimInstance(), CurveName=Frontend_Idle_Bypass)) -> (Output_Get:float)
  ]}
]
variables [FE Custom Idle Is Playing Float {type=number, default=1}]
